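# Post-processes the RankedShade female medium body cosmetic, adding secondary motion to its bags and belt on top of the pose copied from the attached parent mesh.
# An armor Control Rig, fed parachute and aiming states, produces a cached pose.
# Weighted multi-bone constraints then drive bag bases and belt bones from spine, pelvis, thigh and deform bones.
# A crouch-aware bag Control Rig follows, then root-relative spring AnimDynamics simulates six bag bones.
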AnimBlueprint /BRCosmetics/Characters/Player/Female/Medium/Bodies/F_MED_RankedShade/Meshes/F_MED_RankedShade_AnimBP
target_skeleton=&/F_MED_RankedShade_Skeleton
graphs [
 AnimGraph {nodes [
   n1 = "Constraint - bag_r_01_base (PPPPP)"!(
    ComponentPose=Local To Component(LocalPose=Use cached pose 'Control RigDef'())
    Alpha=1
    ConstraintWeights_0=0.2191
    ConstraintWeights_1:float=0.5257
    ConstraintWeights_2:float=0.0809
    ConstraintWeights_3:float=0.002
    ConstraintWeights_4:float=0.1722
   ) -> (Pose) {
    boneToModify {boneName=bag_r_01_base}
    constraintSetup [
     {targetBone {boneName=pelvis}}
     {targetBone {boneName=thigh_twist_01_r}}
     {targetBone {boneName=thigh_r}}
     {targetBone {boneName=deform_glute_r}}
     {targetBone {boneName=deform_groin_r}}
    ]
    constraintWeights [1,1,1,1,1]
   }
   n2 = "Constraint - bag_r_02_base (PPPPP)"!(
    ComponentPose=@n1
    Alpha=1
    ConstraintWeights_0=0.2741
    ConstraintWeights_1:float=0.5859
    ConstraintWeights_2:float=0.0597
    ConstraintWeights_3:float=0.0387
    ConstraintWeights_4:float=0.0415
   ) -> (Pose) {
    boneToModify {boneName=bag_r_02_base}
    constraintSetup [
     {targetBone {boneName=pelvis}}
     {targetBone {boneName=thigh_twist_01_r}}
     {targetBone {boneName=thigh_r}}
     {targetBone {boneName=deform_glute_r}}
     {targetBone {boneName=deform_groin_r}}
    ]
    constraintWeights [1,1,1,1,1]
   }
   n3 = "Constraint - bag_r_03_base (PPPP)"!(
    ComponentPose=@n2
    Alpha=1
    ConstraintWeights_0=0.3144
    ConstraintWeights_1:float=0.4793
    ConstraintWeights_2:float=0.0111
    ConstraintWeights_3:float=0.1953
   ) -> (Pose) {
    boneToModify {boneName=bag_r_03_base}
    constraintSetup [
     {targetBone {boneName=pelvis}}
     {targetBone {boneName=thigh_twist_01_r}}
     {targetBone {boneName=thigh_r}}
     {targetBone {boneName=deform_glute_r}}
    ]
    constraintWeights [1,1,1,1]
   }
   n4 = "Constraint - bag_l_01_base (PPPPP)"!(
    ComponentPose=@n3
    Alpha=1
    ConstraintWeights_0=0.2191
    ConstraintWeights_1:float=0.5257
    ConstraintWeights_2:float=0.0809
    ConstraintWeights_3:float=0.002
    ConstraintWeights_4:float=0.1722
   ) -> (Pose) {
    boneToModify {boneName=bag_l_01_base}
    constraintSetup [
     {targetBone {boneName=pelvis}}
     {targetBone {boneName=thigh_twist_01_l}}
     {targetBone {boneName=thigh_l}}
     {targetBone {boneName=deform_glute_l}}
     {targetBone {boneName=deform_groin_l}}
    ]
    constraintWeights [1,1,1,1,1]
   }
   n5 = "Constraint - bag_l_02_base (PPPPP)"!(
    ComponentPose=@n4
    Alpha=1
    ConstraintWeights_0=0.2741
    ConstraintWeights_1:float=0.5859
    ConstraintWeights_2:float=0.0597
    ConstraintWeights_3:float=0.0387
    ConstraintWeights_4:float=0.0415
   ) -> (Pose) {
    boneToModify {boneName=bag_l_02_base}
    constraintSetup [
     {targetBone {boneName=pelvis}}
     {targetBone {boneName=thigh_twist_01_l}}
     {targetBone {boneName=thigh_l}}
     {targetBone {boneName=deform_glute_l}}
     {targetBone {boneName=deform_groin_l}}
    ]
    constraintWeights [1,1,1,1,1]
   }
   n6 = "Constraint - bag_l_03_base (PPPP)"!(
    ComponentPose=@n5
    Alpha=1
    ConstraintWeights_0=0.3138
    ConstraintWeights_1:float=0.4619
    ConstraintWeights_2:float=0.0091
    ConstraintWeights_3:float=0.2152
   ) -> (Pose) {
    boneToModify {boneName=bag_l_03_base}
    constraintSetup [
     {targetBone {boneName=pelvis}}
     {targetBone {boneName=thigh_twist_01_l}}
     {targetBone {boneName=thigh_l}}
     {targetBone {boneName=deform_glute_l}}
    ]
    constraintWeights [1,1,1,1]
   }
   n7 = "Constraint - def_belt_f_m (PPPPPPPP)"!(
    ComponentPose=@n6
    Alpha=1
    ConstraintWeights_0=0.0065
    ConstraintWeights_1:float=0.001
    ConstraintWeights_2:float=0.0867
    ConstraintWeights_3:float=0.8867
    ConstraintWeights_4:float=0.0049
    ConstraintWeights_5:float=0.004
    ConstraintWeights_6:float=0.0052
    ConstraintWeights_7:float=0.0049
   ) -> (Pose) {
    boneToModify {boneName=def_belt_f_m}
    constraintSetup [
     {targetBone {boneName=spine_04}}
     {targetBone {boneName=spine_03}}
     {targetBone {boneName=spine_02}}
     {targetBone {boneName=pelvis}}
     {targetBone {boneName=thigh_twist_01_l}}
     {targetBone {boneName=thigh_twist_01_r}}
     {targetBone {boneName=deform_groin_l}}
     {targetBone {boneName=deform_groin_r}}
    ]
    constraintWeights [1,1,1,1,1,1,1,1]
   }
   n8 = "Constraint - def_belt_b_m (PPPPP)"!(
    ComponentPose=@n7
    Alpha=1
    ConstraintWeights_0=0.0411
    ConstraintWeights_1:float=0.0504
    ConstraintWeights_2:float=0.3557
    ConstraintWeights_3:float=0.1215
    ConstraintWeights_4:float=0.4312
   ) -> (Pose) {
    boneToModify {boneName=def_belt_b_m}
    constraintSetup [
     {targetBone {boneName=spine_04}}
     {targetBone {boneName=spine_03}}
     {targetBone {boneName=spine_02}}
     {targetBone {boneName=spine_01}}
     {targetBone {boneName=pelvis}}
    ]
    constraintWeights [1,1,1,1,1]
   }
   n9 = Control Rig!(Source=Component To Local(ComponentPose=@n8), Alpha, is Crouch:bool=Get bIsCrouching()) -> (Pose) {controlRigAssetReference {blueprintRigClass=&/F_MED_RankedShade_CtrlRig_Bags.F_MED_RankedShade_CtrlRig_Bags_C}}
   n10 = Anim Dynamics - Bone: bag_r_01(ComponentPose=Local To Component(LocalPose=@n9), Alpha=1, ExternalForce="0,0,0") {
    angularDampingOverride=0.9999
    angularSpringConstant=500
    bAngularSpring=true
    bOverrideAngularDamping=true
    bUseGravityOverride=true
    linearDampingOverride=1
    linearSpringConstant=50
    simulationSpace=RootRelative
    alphaBoolBlend {blendOption=Linear}
    boundBone {boneName=bag_r_01}
    componentAppliedLinearAccClamp=<3000 3000 3000>
    componentLinearAccScale=<0.4 0.4 0.4>
    componentLinearVelScale=<0.4 0.4 0.4>
    gravityOverride {z=-980}
    physicsBodyDefinitions [{
      boundBone {boneName=bag_r_01}
      boxExtents=<20 25 20>
      constraintSetup {angularLimitsMax=<2 0 5>, angularLimitsMin=<-2 -10 -5>, angularTarget=<1 0 0>}
      localJointOffset=<2 0 0>
     }]
   }
   n11 = Anim Dynamics - Bone: bag_r_02(ComponentPose=@n10, Alpha=1, ExternalForce="0,0,0") {
    angularDampingOverride=0.9999
    angularSpringConstant=500
    bAngularSpring=true
    bOverrideAngularDamping=true
    bUseGravityOverride=true
    linearDampingOverride=1
    linearSpringConstant=50
    simulationSpace=RootRelative
    alphaBoolBlend {blendOption=Linear}
    boundBone {boneName=bag_r_02}
    componentAppliedLinearAccClamp=<3000 3000 3000>
    componentLinearAccScale=<0.4 0.4 0.4>
    componentLinearVelScale=<0.4 0.4 0.4>
    gravityOverride {z=-980}
    physicsBodyDefinitions [{
      boundBone {boneName=bag_r_02}
      boxExtents=<20 25 20>
      constraintSetup {angularLimitsMax=<2 0 5>, angularLimitsMin=<-2 -10 -5>, angularTarget=<1 0 0>}
      localJointOffset=<2 0 0>
     }]
   }
   n12 = Anim Dynamics - Bone: bag_r_03(ComponentPose=@n11, Alpha=1, ExternalForce="0,0,0") {
    angularDampingOverride=0.9999
    angularSpringConstant=500
    bAngularSpring=true
    bOverrideAngularDamping=true
    bUseGravityOverride=true
    linearDampingOverride=1
    linearSpringConstant=50
    simulationSpace=RootRelative
    alphaBoolBlend {blendOption=Linear}
    boundBone {boneName=bag_r_03}
    componentAppliedLinearAccClamp=<3000 3000 3000>
    componentLinearAccScale=<0.4 0.4 0.4>
    componentLinearVelScale=<0.4 0.4 0.4>
    gravityOverride {z=-980}
    physicsBodyDefinitions [{
      boundBone {boneName=bag_r_03}
      boxExtents=<20 25 20>
      constraintSetup {angularLimitsMax=<2 0 5>, angularLimitsMin=<-2 -10 -5>, angularTarget=<1 0 0>}
      localJointOffset=<2 0 0>
     }]
   }
   n13 = Anim Dynamics - Bone: bag_l_01(ComponentPose=@n12, Alpha=1, ExternalForce="0,0,0") {
    angularDampingOverride=0.9999
    angularSpringConstant=500
    bAngularSpring=true
    bOverrideAngularDamping=true
    bUseGravityOverride=true
    linearDampingOverride=1
    linearSpringConstant=50
    simulationSpace=RootRelative
    alphaBoolBlend {blendOption=Linear}
    boundBone {boneName=bag_l_01}
    componentAppliedLinearAccClamp=<3000 3000 3000>
    componentLinearAccScale=<0.4 0.4 0.4>
    componentLinearVelScale=<0.4 0.4 0.4>
    gravityOverride {z=-980}
    physicsBodyDefinitions [{
      boundBone {boneName=bag_l_01}
      boxExtents=<20 25 20>
      constraintSetup {angularLimitsMax=<2 0 5>, angularLimitsMin=<-2 -10 -5>, angularTarget=<1 0 0>}
      localJointOffset=<2 0 0>
     }]
   }
   n14 = Anim Dynamics - Bone: bag_l_02(ComponentPose=@n13, Alpha=1, ExternalForce="0,0,0") {
    angularDampingOverride=0.9999
    angularSpringConstant=500
    bAngularSpring=true
    bOverrideAngularDamping=true
    bUseGravityOverride=true
    linearDampingOverride=1
    linearSpringConstant=50
    simulationSpace=RootRelative
    alphaBoolBlend {blendOption=Linear}
    boundBone {boneName=bag_l_02}
    componentAppliedLinearAccClamp=<3000 3000 3000>
    componentLinearAccScale=<0.4 0.4 0.4>
    componentLinearVelScale=<0.4 0.4 0.4>
    gravityOverride {z=-980}
    physicsBodyDefinitions [{
      boundBone {boneName=bag_l_02}
      boxExtents=<20 25 20>
      constraintSetup {angularLimitsMax=<2 0 5>, angularLimitsMin=<-2 -10 -5>, angularTarget=<1 0 0>}
      localJointOffset=<2 0 0>
     }]
   }
   n15 = Anim Dynamics - Bone: bag_l_03(ComponentPose=@n14, Alpha=1, ExternalForce="0,0,0") {
    angularDampingOverride=0.9999
    angularSpringConstant=500
    bAngularSpring=true
    bOverrideAngularDamping=true
    bUseGravityOverride=true
    linearDampingOverride=1
    linearSpringConstant=50
    simulationSpace=RootRelative
    alphaBoolBlend {blendOption=Linear}
    boundBone {boneName=bag_l_03}
    componentAppliedLinearAccClamp=<3000 3000 3000>
    componentLinearAccScale=<0.4 0.4 0.4>
    componentLinearVelScale=<0.4 0.4 0.4>
    gravityOverride {z=-980}
    physicsBodyDefinitions [{
      boundBone {boneName=bag_l_03}
      boxExtents=<20 25 20>
      constraintSetup {angularLimitsMax=<2 0 5>, angularLimitsMin=<-2 -10 -5>, angularTarget=<1 0 0>}
      localJointOffset=<2 0 0>
     }]
   }
   Output Pose(Result=Component To Local(ComponentPose=@n15)) {layerGroup=None}
   n16 = Control Rig!(
    Source=Copy Pose From Mesh() {bUseAttachedParent=true}
    Alpha
    IsParachuteOpenedAndCharUpRight:bool=Get bParachuteOpenedAndStandingUpright() -> (bParachuteOpenedAndStandingUpright:bool)
    isParachuteOpen:bool=Get bIsParachuteOpened()
    isAiming:bool=Get bIsTargeting()
   ) -> (Pose) {controlRigAssetReference {blueprintRigClass=&/F_MED_RankedShade_CtrlRig_Armor.F_MED_RankedShade_CtrlRig_Armor_C}}
   Save cached pose 'Control RigDef'(Pose=@n16) {cacheName=Control RigDef}
  ]}
 EventGraph {nodes [Event BlueprintUpdateAnimation()]}
]
variables []
